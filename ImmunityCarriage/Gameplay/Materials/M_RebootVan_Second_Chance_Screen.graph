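# Emissive screen material for the Reboot Van (Second Chance machine), showing cooldown, interaction and spawning states.
# Cooldown and Spawning parameters blend between a blinking, time-filled battery in the fail color, a logo with radial interaction fill, and a spinning loader.
# Scanlines, periodic UV jitter, vertex-color mirroring and edge/fresnel dimming style the screen, with cheaper Low-quality branches.
# Time of day, an interaction flash and a post-cooldown-start blackout scale brightness.
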Material /ImmunityCarriage/Gameplay/Materials/M_RebootVan_Second_Chance_Screen
graph {
 properties {
  bUsedWithMeshDeformer=true
  bUsedWithNanite=true
  bUsedWithSkeletalMesh=true
  pixelDepthOffsetMode=PDOM_Legacy
  usedShadingModels=Default Lit
 }
 nodes [
  Constant() {r=0.5}
  Constant() {r=0.1}
  n1 = Constant() {r=1}
  n2 = ScalarParameter() {defaultValue=1, parameterName=HeightScale}
  n3 = Panner(Coordinate=Multiply(
    A=TextureCoordinate() {vTiling=32}
    B=MaterialFunctionCall(X=@n1, Y=@n2) {materialFunction=&MakeFloat2}.Result
   )) {speedX=0.5, speedY=0.7}
  n4 = QualitySwitch(
   Default=Multiply(A=TextureSample(UVs=@n3) {texture=/Game/Missions/Primary/LaunchTheBalloon/Textures/T_TV_Scanlines}.RGB) {constB=0.05}
   Low=Constant()
  )
  n5 = VectorParameter() {parameterName=ScreenColor, defaultValue=<rgba 0.1205 0.7309 1.5 0>}
  n6 = VectorParameter() {parameterName=ScreenColorFail, defaultValue=<rgba 1.5 0.1677 0.108 0>}
  n7 = ScalarParameter() {defaultValue=1, parameterName=Cooldown}
  n8 = Time()
  n9 = ScalarParameter() {defaultValue=1, parameterName=CooldownTime}
  n10 = ScalarParameter() {defaultValue=4, parameterName=CooldownDuration}
  n11 = Clamp(None=Divide(A=Subtract(A=@n8, B=@n9), B=@n10))
  n12 = ScalarParameter() {defaultValue=0.5, parameterName=CooldownEndDuration}
  n13 = Multiply(A=@n11, B=Add(A=Constant() {r=1}, B=Divide(A=@n12, B=@n10)) {desc=Duration to show filled battery}) {constB=1.1}
  n14 = Floor(None=@n13)
  n15 = Time()
  n16 = Round(None=Add(A=Sine(None=@n15) {desc=Blink battery icon, period=0.5}))
  n17 = Frac(None=ComponentMask(None=TextureCoordinate() {vTiling=2.2}) {g=true})
  n18 = Multiply(A=Time()) {constB=0.23}
  n19 = Floor(None=Multiply(A=Frac(None=@n18)) {constB=1.1})
  n20 = Multiply(A=Floor(None=Multiply(A=@n17) {constB=3}), B=@n19) {constA=0.3}
  n21 = Clamp(None=Multiply(B=Multiply(A=@n20, B=@n19)) {constA=-0.03}) {minDefault=-1}
  n22 = Constant()
  n23 = Add(A=QualitySwitch(Default=AppendVector(A=@n21, B=@n22), Low=Constant()), B=TextureCoordinate())
  n24 = ComponentMask(None=VectorParameter() {parameterName=TextureScale, defaultValue=<rgba 1.5 1.125 0 0>}.RGB) {g=true, r=true}
  n25 = VectorParameter() {parameterName=TextureOffset, defaultValue=<rgba -0.25 -0.12 0 0>}
  n26 = MaterialFunctionCall(Float2=Add(A=Multiply(A=@n23, B=@n24), B=ComponentMask(None=@n25.RGB) {g=true, r=true})) {materialFunction=&BreakOutFloat2Components}
  n27 = AppendVector(A=LinearInterpolate(A=@n26.R, B=OneMinus(None=@n26.R), Alpha=VertexColor().G), B=@n26.G)
  n28 = TextureSample(UVs=@n27) {texture=/Game/Athena/Items/EnvironmentalItems/SCMachine/FX/T_Reboot_Battery_FX}
  n29 = LinearInterpolate(
   A=ScalarParameter() {defaultValue=1.25, parameterName=BatteryGradMax}
   B=ScalarParameter() {defaultValue=-0.325, parameterName=BatteryGradMin}
   Alpha=ComponentMask(None=@n27) {r=true}
  ) {constA=-0.5, constB=1.5}
  n30 = Floor(None=Clamp(None=Add(A=@n13, B=@n29)))
  n31 = Add(A=Multiply(A=Clamp(None=@n16, Min=@n14), B=Multiply(A=@n28.R, B=@n30)) {constA=1}, B=@n28.G)
  n32 = ScalarParameter() {defaultValue=1, parameterName=Cooldown}
  n33 = Multiply(A=Multiply(A=Multiply(A=@n14, B=@n31)) {constB=2.5}, B=@n32)
  n34 = TextureSample(UVs=MaterialFunctionCall(UVs=@n27, Texture Scale=Constant2Vector() {g=1, r=-1}) {materialFunction=&ScaleUVsByCenter}.UVs) {texture=/Game/Athena/Items/EnvironmentalItems/SCMachine/FX/T_Reboot_Logo_FX}
  n35 = Multiply(A=OneMinus(None=ScalarParameter() {defaultValue=1, parameterName=Spawning}), B=@n34.B)
  n36 = Add(A=Multiply(A=@n31) {constB=0.7}, B=Multiply(A=@n28.R) {constB=0.3})
  n37 = Clamp(None=OneMinus(None=LinearInterpolate(A=@n35, B=@n36, Alpha=@n32)), Min=ScalarParameter() {defaultValue=0.33, parameterName=DarkColor})
  n38 = QualitySwitch(Default=@n37, Low=Multiply(A=@n37, B=Constant() {desc=Darken BG on Low, r=0.4}))
  n39 = ScalarParameter() {defaultValue=1, parameterName=Cooldown}
  n40 = VectorParameter() {parameterName=FillColor, defaultValue=<rgba 0.2805 3 2.9403 0>}
  n41 = MaterialFunctionCall(Float3=@n34.RGB) {materialFunction=&BreakOutFloat3Components}
  n42 = OneMinus(None=@n41.G)
  n43 = Subtract(A=Time(), B=ScalarParameter() {parameterName=InteractTime})
  n44 = Add(A=Multiply(A=@n43, B=ScalarParameter() {parameterName=Multiplier}), B=ScalarParameter() {parameterName=BaseTime})
  n45 = ScalarParameter() {parameterName=NewMultiWeight}
  n46 = ScalarParameter() {defaultValue=4, parameterName=InteractDuration}
  n47 = Clamp(None=Divide(A=LinearInterpolate(A=@n43, B=@n44, Alpha=@n45), B=@n46))
  n48 = Multiply(A=OneMinus(None=Floor(None=@n47)), B=@n47) {desc=Go dark when full}
  n49 = MaterialFunctionCall(
   UVs=MaterialFunctionCall(UVs=@n27, Texture Scale=Constant2Vector() {g=1, r=-1}) {materialFunction=&ScaleUVsByCenter}.UVs
   "Rotation Angle (0-1)"=ScalarParameter() {defaultValue=0.2, parameterName=Origin_0-1}
  ) {materialFunction=&CustomRotator}
  n50 = OneMinus(None=MaterialFunctionCall(Vector or UVs=Subtract(A=@n49.Rotated Values) {constB=0.5}) {materialFunction=&VectorToRadialValue}.Vector Converted to Angle)
  n51 = Clamp(None=Floor(None=Add(A=@n48, B=@n50)), Max=@n41.R)
  n52 = Constant() {r=0.5}
  n53 = Floor(None=Add(A=@n48, B=Divide(A=@n52, B=@n46)) {constB=0.02})
  n54 = Multiply(A=Time(), B=ScalarParameter() {defaultValue=1.5, parameterName=SpinnerSpeed})
  n55 = LinearInterpolate(
   A=Add(A=Multiply(A=@n42, B=@n51), B=Multiply(A=@n41.G, B=@n53) {desc=Fill up pip near end})
   B=TextureSample(UVs=Rotator(Coordinate=@n27, Time=@n54) {speed=-10}) {texture=/Game/Athena/Items/EnvironmentalItems/SCMachine/FX/T_Loading_Spinner}.G
   Alpha=ScalarParameter() {defaultValue=1, parameterName=Spawning}
  )
  n56 = Add(
   A=Multiply(A=LinearInterpolate(A=@n5.RGB, B=@n6.RGB, Alpha=@n7), B=Add(A=@n33, B=@n38)) {constB=0.3}
   B=Multiply(A=OneMinus(None=@n39), B=Multiply(A=@n40.RGB, B=@n55) {constA=1, constB=0.2})
  )
  n57 = ScalarParameter() {defaultValue=0.5, parameterName=FrasnelDarkenExp}
  n58 = Multiply(
   A=SphereMask(A=TextureCoordinate(), B=Constant() {r=0.5}, Radius=ScalarParameter() {defaultValue=0.74, parameterName=EdgeDimRadius}) {hardnessPercent=0}
   B=OneMinus(None=Fresnel(ExponentIn=@n57) {baseReflectFraction=0, exponent=0.5})
  )
  n59 = Constant() {r=0.5}
  n60 = CollectionParameter() {
   collection=&FortniteMaterialParameters
   parameterName=FortniteActiveTimeOfDay
  }
  n61 = Multiply(
   A=Multiply(A=Add(A=@n4, B=@n56), B=QualitySwitch(Default=@n58, Low=@n59))
   B=LinearInterpolate(B=ScalarParameter() {defaultValue=0.55, parameterName=NightBrightness}, Alpha=ComponentMask(None=@n60) {a=true}) {constA=1}
  )
  n62 = Subtract(A=Time(), B=ScalarParameter() {parameterName=InteractTime})
  n63 = Add(A=Multiply(A=@n62, B=ScalarParameter() {parameterName=Multiplier}), B=ScalarParameter() {parameterName=BaseTime})
  n64 = ScalarParameter() {parameterName=NewMultiWeight}
  n65 = Clamp(None=Divide(A=LinearInterpolate(A=@n62, B=@n63, Alpha=@n64), B=ScalarParameter() {defaultValue=0.5, parameterName=FlashDuration}))
  n66 = Add(
   A=OneMinus(None=@n11)
   B=Divide(
    A=ScalarParameter() {defaultValue=1, parameterName=BlackScreenDuration}
    B=ScalarParameter() {defaultValue=4, parameterName=CooldownDuration}
   )
  ) {constB=0.15}
  n67 = Multiply(
   A=Add(
    A=MaterialFunctionCall!(In=@n65) {materialFunction=/Game/Effects/Fort_Effects/Materials/_Functions/MF_Smoothstep_EaseOut}.Result
    B=ScalarParameter() {defaultValue=1, parameterName=ScreenBrightness}
   )
   B=OneMinus(None=Floor(None=@n66))
  )
  MaterialFunctionCall(Emissive=Multiply(A=Multiply(A=@n61, B=@n67), B=ScalarParameter() {defaultValue=1, parameterName=ScaleEmissiveBrightness})) {materialFunction=&MF_FX_EmissiveAdjustments}
  Time() {bOverride_Period=true, period=4}
  Time() {bOverride_Period=true, period=4}
 ]
}
