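# Physics asset for the RebootVan vehicle mesh in the ImmunityCarriage gameplay content.
# It defines a single physics body on the frame bone.
# It has no constraints, so the body is not jointed to any other part.
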PhysicsAsset /ImmunityCarriage/Gameplay/Meshes/RebootVan_Veh_Physics
bodies {frame []}
constraints []
